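# Unlit translucent flame material for Niagara sprite and mesh particles in the SummitReedMolt emote FX, derived from Curie fire.
# Dynamic parameters pick a packed SubUV flipbook channel, which panning tiling noise modulates to drive curve-atlas color and opacity.
# Opacity fades by view angle and scene depth, with cheaper color and opacity on Low quality.
# WPO tilts particles toward an ignition vector (built from X alone), optionally snapping to distance-field surfaces.
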Material /BRCosmetics/Animation/Game/MainPlayer/Emotes/SummitReedMolt/FX/M_Curie_SubUVFlames_Packed_AngleFade
graph {
 properties {
  bCastRayTracedShadows=false
  bUsedWithNiagaraMeshParticles=true
  bUsedWithNiagaraSprites=true
  blendMode=BLEND_Translucent
  pixelDepthOffsetMode=PDOM_Legacy
  shadingModel=MSM_Unlit
  usedShadingModels=Unlit
 }
 nodes [
  n1 = DynamicParameter() {parameterIndex=2, paramNames [Ignition Normal.X, Ignition Normal.Y, Ignition Normal.Z, Size]}
  n2 = DynamicParameter() {parameterIndex=1, paramNames [Sprite_RandomOffset, Sprite_TimeOffset, Detail_Intensity, ""]}
  n3 = Add(
   A=Multiply(A=TextureCoordinate(), B=@n1.Size)
   B=MaterialFunctionCall(X=@n2.Sprite_RandomOffset, Y=@n2.Sprite_TimeOffset) {materialFunction=&MakeFloat2}.Result
  )
  n4 = ScalarParameter() {defaultValue=1, parameterName=Detail TexScale, sliderMax=1}
  n5 = Time()
  n6 = Multiply(A=@n2.Sprite_RandomOffset) {constB=25}
  n7 = Constant()
  n8 = Multiply(A=@n1.Size, B=ScalarParameter() {defaultValue=1.3, parameterName=Detail PanSpeed, sliderMax=1})
  n9 = TextureSample(UVs=Panner(Coordinate=Multiply(A=@n3, B=@n4) {constB=0.4}, Time=Add(A=@n5, B=@n6), Speed=AppendVector(A=@n7, B=@n8)) {speedY=0.05}) {samplerType=SAMPLERTYPE_LinearGrayscale, texture=/BRCosmetics/Weapons/FORT_Melee/Pickaxe_Cube_Ninja_Male/FX/T_FluidFlameTile}
  n10 = ComponentMask(None=Multiply(
    A=DynamicParameter() {parameterIndex=1, paramNames [Sprite_RandomOffset, Sprite_TimeOffset, Detail_Intensity, ""]}.Detail_Intensity
    B=Power(Base=@n9.R) {constExponent=0.5}
   ) {constA=0.5}) {r=true}
  CurveAtlasRowParameter(CurveTime=@n10) {
   atlas=/Game/Curie/Effects/temp/Curves/CA_Curie_Fire_Visuals
   curve=/Game/Curie/Effects/temp/Curves/Curve_Fiery_Orange
   defaultValue=1
   parameterName=CurveAtlas_fire
  }
  n11 = ParticleColor()
  n12 = NamedRerouteUsage()
  n13 = Multiply(
   A=@n12
   B=LinearInterpolate(
    A=ScalarParameter() {defaultValue=0.8, parameterName=Detail Min, sliderMax=3, sliderMin=-1}
    B=ScalarParameter() {defaultValue=1.5, parameterName=Detail Max, sliderMax=3, sliderMin=-1}
    Alpha=NamedRerouteUsage()
   ) {constA=0.8, constB=2}
  )
  n14 = Power(
   Base=QualitySwitch(Default=@n13, Low=@n12, Epic=@n13) {desc=quality Opactiy}
   Exp=ScalarParameter() {defaultValue=1.508, group=Output Color, parameterName=Opacity Exponent, sliderMax=3, sortPriority=8}
  )
  n15 = ScalarParameter() {defaultValue=3, group=Output Color, parameterName=Opacity Scale, sliderMax=3, sortPriority=8}
  n16 = Saturate(None=Multiply(A=@n11.A, B=Multiply(A=@n14, B=@n15) {constB=3}))
  n17 = CurveAtlasRowParameter(CurveTime=@n13) {
   atlas=&CA_Curie_Fire_Visuals
   curve=/Game/Curie/Effects/temp/Curves/Curve_Fiery_B
   defaultValue=8
   parameterName=Curves
  }
  n18 = CurveAtlasRowParameter(CurveTime=@n13) {
   atlas=/Game/Curie/Effects/temp/Curves/CA_NonCurie_FireVisuals
   curve=&Curve_Fiery_B
   parameterName=NonCurieCurves
  }
  n19 = StaticSwitchParameter(True=@n17, False=@n18) {defaultValue=true, parameterName=is Curie}
  n20 = LinearInterpolate(A=@n19, B=Divide(A=@n19, B=@n16), Alpha=ScalarParameter() {defaultValue=0.45, group=Output Color, parameterName=UnMult, sliderMax=1})
  n21 = QualitySwitch(Default=@n20, Low=Multiply(A=VectorParameter() {parameterName=Low Flat Color, defaultValue=<rgba 1 0.2935 0.1029 1>}.RGB, B=@n12), Epic=@n20) {desc=quality RGB}
  n22 = ScalarParameter() {defaultValue=2, group=Output Color, parameterName=Final Color Brightness, sliderMax=50, sortPriority=9}
  MaterialFunctionCall(Emissive=Multiply(A=@n11.RGB, B=Multiply(A=@n21, B=@n22))) {materialFunction=&MF_FX_EmissiveAdjustments}
  n23 = TextureObjectParameter() {group=Flame Channels, parameterName=Texture, sortPriority=1, texture=/Game/Effects/Fort_Effects/Textures/Fire/2DGasTorch_Flipbk_packed}
  n24 = MaterialFunctionCall(Float2=TextureCoordinate()) {materialFunction=&BreakOutFloat2Components}
  n25 = Power(Base=@n24.G, Exp=ScalarParameter() {defaultValue=1.248, parameterName=Vertical Exponent, sliderMax=3}) {constExponent=1}
  n26 = DynamicParameter() {paramNames [Temperature Channel Scale, Packed Channel, SubUV_Images, SubUV_Frame]}
  n27 = MaterialFunctionCall(Float3=MaterialFunctionCall!(
    Texture=@n23
    UVs=MaterialFunctionCall(X=@n24.R, Y=@n25) {materialFunction=&MakeFloat2}.Result
    SubImages=AppendVector(A=@n26.SubUV_Images, B=@n26.SubUV_Images)
    Frame=@n26.SubUV_Frame
   ) {materialFunction=/Engine/Functions/Engine_MaterialFunctions03/Texturing/SubUV_Function}.RGB) {materialFunction=&BreakOutFloat3Components}
  n28 = DynamicParameter() {desc=G - packed channel switch/weight, paramNames [Temperature Channel Scale, Packed Channel, SubUV_Images, SubUV_Frame]}
  n29 = Multiply(
   A=Multiply(
    A=@n28.Temperature Channel Scale
    B=If(A=@n28.Packed Channel, "A > B"=@n27.B, "A == B"=@n27.G, "A < B"=@n27.R) {constB=1, desc=pick flipbook channels, equalsThreshold=0.1}
   )
   B=ScalarParameter() {defaultValue=1, group=Flame Channels, parameterName=Temperature Channel Scale, sliderMax=10, sortPriority=3}
  )
  n30 = ScalarParameter() {defaultValue=2, group=Flame Channels, parameterName=Temperature Exponent, sliderMax=3, sortPriority=5}
  NamedRerouteDeclaration(None=Multiply(
    A=Power(Base=@n29, Exp=@n30)
    B=ScalarParameter() {defaultValue=4, group=Flame Channels, parameterName=Temperature Scale, sliderMax=25, sortPriority=5}
   )) {name=Flipbook Flame Texture, nodeColor=<rgba 0 0.6016 1 1>}
  n31 = ParticlePositionWS()
  n32 = DynamicParameter() {parameterIndex=2, paramNames [Ignition Normal.X, Ignition Normal.Y, Ignition Normal.Z, Size]}
  n33 = MaterialFunctionCall(X=@n32.Ignition Normal.X, Y=@n32.Ignition Normal.X, Z=@n32.Ignition Normal.X) {materialFunction=&MakeFloat3}
  n34 = QualitySwitch(Default=DistanceFieldGradient!(World Position=@n31), Low=@n33.Result, Medium=@n33.Result) {desc=DF on High/Epic}
  n35 = StaticBoolParameter() {parameterName=Use DistanceField}
  n36 = StaticSwitch(True=FeatureLevelSwitch(Default=@n34, ES3_1=@n33.Result, SM6=@n34) {desc=disable for ES3.1}, False=@n33.Result, Value=@n35)
  n37 = Subtract(A=WorldPosition() {worldPositionShaderOffset=WPT_ExcludeAllShaderOffsets}.XYZ, B=@n31)
  n38 = Subtract(
   A=MaterialFunctionCall!(VectorToRotate=@n37, LookAtVector=Normalize(VectorInput=@n36), RestingVector=Constant3Vector() {constant=<rgba 0 0 1 0>}) {materialFunction=/Engine/Functions/Engine_MaterialFunctions01/Vectors/RotateVector}.Result
   B=@n37
  )
  n39 = DistanceFieldGradient(World Position=@n31)
  n40 = DistanceToNearestSurface!(World Position=@n31)
  n41 = Add(A=@n38, B=Multiply(A=Normalize(VectorInput=@n39), B=Multiply(A=@n40) {constB=-1})) {constB=0}
  n42 = FeatureLevelSwitch(Default=@n41, ES3_1=@n38, SM6=@n41) {desc=disable for ES3.1}
  n43 = ScalarParameter() {defaultValue=0.3, parameterName=Stick Min, sliderMax=1}
  n44 = ScalarParameter() {defaultValue=1, parameterName=Stick Max, sliderMax=1}
  n45 = Power(
   Base=MaterialFunctionCall(Float2=TextureCoordinate()) {materialFunction=&BreakOutFloat2Components}.G
   Exp=ScalarParameter() {defaultValue=0.76, parameterName=Stick Exponent, sliderMax=5}
  ) {constExponent=0.75}
  NamedRerouteDeclaration(None=LinearInterpolate(A=Constant3Vector(), B=StaticSwitch(True=@n42, False=@n38, Value=@n35), Alpha=LinearInterpolate(A=@n43, B=@n44, Alpha=@n45) {constA=0.3})) {name=WPO DF Aligned Edge, nodeColor=<rgba 1 0.0938 0 1>}
  NamedRerouteDeclaration(None=@n10) {name=Tiling Flame Texture, nodeColor=<rgba 1 0 0.3516 1>}
  n46 = ScalarParameter() {defaultValue=1, parameterName=AngleFadeSensitivity}
  n47 = DepthFade(
   Opacity=Multiply(
    A=@n16
    B=MaterialFunctionCall("FadeSensitivity (optional)"=@n46) {materialFunction=&FadeBasedOnViewAngle}.Result
   )
   FadeDistance=ScalarParameter() {defaultValue=55, parameterName=FadeDistance, sliderMax=300}
  )
  Reroute(None=QualitySwitch(Default=@n47, Low=@n16, Epic=@n47) {desc=quality Opactiy})
  StaticSwitch(True=NamedRerouteUsage(), Value=StaticBoolParameter() {defaultValue=true, parameterName=UseWPO})
  TextureSample(UVs=TextureCoordinate()) {desc=channel debug, texture=/Game/Effects/Fort_Effects/Textures/Debug/RGB_Debug}
 ]
}
