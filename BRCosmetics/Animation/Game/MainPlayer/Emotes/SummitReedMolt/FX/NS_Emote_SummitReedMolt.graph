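# Niagara system for the SummitReedMolt player emote, layering ambient atmosphere, flashing lights and pyrotechnics.
# Ambient_Condensed merges many spawn groups into one emitter, switching size, placement, fade and System.LightColor tint per group.
# Flashing_Lights spawns delayed, randomly culled sprite grids; alpha adapts to platform, quality and time of day, with front-end overrides in places.
# Embers and Pyrotechnics are disabled, with spawn rates gated by staged PyroStage values and User.Pyrotechnics.
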NiagaraSystem /BRCosmetics/Animation/Game/MainPlayer/Emotes/SummitReedMolt/FX/NS_Emote_SummitReedMolt
dependencies {
 usedDataInterfaces [
  &NiagaraDataInterfaceSpriteRendererInfo
  &NiagaraDataInterfaceCurve
  &NiagaraDataInterfaceVector2DCurve
  &NiagaraDataInterfaceArrayFloat
  &NiagaraDataInterfacePlatformSet
  &NiagaraDataInterfaceColorCurve
  &NiagaraDataInterfaceArrayFloat3
 ]
 usedDynamicInputs [
  &MakeFloatFromVector2d
  &MakeCustomIntFromBool
  &Lerp_Vector
  &MakeLinearColorFromVectorAndFloat
 ]
 usedModules [
  &SystemState
  &BlendSpaceHelper
  &self:SystemSpawnScript.SystemScriptSource.SystemScriptGraph.NiagaraNodeAssignment_0.SetVariables_#1
  &self:SystemSpawnScript.SystemScriptSource.SystemScriptGraph.NiagaraNodeAssignment_1.SetVariables_#2
  &InterpolateOverTime
  &self:SystemSpawnScript.SystemScriptSource.SystemScriptGraph.NiagaraNodeAssignment_2.SetVariables_#3
 ]
}
emitters {
 Ambient_Condensed {
  emitterUpdateScript [
   &EmitterState(Life Cycle Mode=::NewEnumerator1, Loop Duration=9.3, UseLoopDelay=false)
   &SpawnRate(SpawnRate=4, Use Spawn Probability=false) {enabled=false}
   &SpawnRate(Spawn Probability=0.7, SpawnGroup=1, SpawnRate=3.5, Use Spawn Probability=true) {module=SpawnRate001}
   &SpawnRate(Spawn Probability=0.7, SpawnGroup=2, SpawnRate=3.5, Use Spawn Probability=true) {module=SpawnRate002}
   &SpawnBurst_Instantaneous(Spawn Count=3, Spawn Group=3, Use Spawn Probability=false, UseLoopCountLimit=false) {module=SpawnBurst_Instantaneous001}
   &SpawnRate(Spawn Probability=0.75, SpawnGroup=3, SpawnRate=2.2, Use Spawn Probability=true) {module=SpawnRate003}
   &SpawnBurst_Instantaneous(Spawn Count=3, Spawn Group=4, Use Spawn Probability=false, UseLoopCountLimit=true) {module=SpawnBurst_Instantaneous002}
   &SpawnRate(Spawn Probability=0.75, SpawnGroup=4, SpawnRate=1.8, Use Spawn Probability=true) {module=SpawnRate004}
   &SpawnBurst_Instantaneous(Spawn Group=5, Use Spawn Probability=false, UseLoopCountLimit=false) {module=SpawnBurst_Instantaneous003}
   &SpawnRate(SpawnGroup=6, Use Spawn Probability=true) {module=SpawnRate005}
   &SpawnBurst_Instantaneous(Spawn Group=7, Use Spawn Probability=false, UseLoopCountLimit=false) {module=SpawnBurst_Instantaneous004}
   &SpawnBurst_Instantaneous(Use Spawn Probability=false, UseLoopCountLimit=false)
  ]
  particleSpawnScript [
   /SetVariables(Particles.VisibilityTag=&SpawnGroup()) {isSetParameters=true}
   &FN_Sprite_Attributes(Create Sprite Rotation Attribute=true, Write to Module Namespace=true)
   &InitializeParticle(
    Color=&MergeHelper_ColorSwitch(
     0=<rgba 0.9062 0.9062 0.9062 0.3>
     1=<rgba 1 1 0.9667 0.02>
     2=<rgba 1.15 1.1385 1.1385 0.1>
     3=<rgba 0.989 1 0.9588 0.11>
     4=<rgba 0.875 0.875 0.875 0.09>
     5=<rgba 0.6979 0.6979 0.6979 0.1>
     6=<rgba 1 1 1 0.3>
     7=<rgba 0.6979 0.6979 0.6979 0.2>
     Index=&SpawnGroup()
     Num=::NewEnumerator6
    )
    Lifetime=&MergeHelper_FloatSwitch(
     0=&RandomRangeFloat(Minimum=0.5, Maximum=1)
     1=&RandomRangeFloat(Minimum=2.5, Maximum=3.5)
     2=&RandomRangeFloat(Minimum=2.5, Maximum=3.5)
     3=&RandomRangeFloat(Minimum=5, Maximum=3)
     4=&RandomRangeFloat(Minimum=5, Maximum=3)
     5=9.3
     6=&RandomRangeFloat(Minimum=3, Maximum=5)
     7=9.3
     Index=&SpawnGroup()
     Num=::NewEnumerator6
    )
    Mesh Scale=&Multiply_VectorByFloat(
     Vector=&RandomRangeVector(Minimum=<1 1 2>, Maximum=<1 1 3>)
     Float=&RandomRangeFloat(Minimum=0.15, Maximum=0.3)
    )
    Mesh Scale Mode=::NewEnumerator4
    Position Offset=&MergeHelper_VectorSwitch(
     0=<0 0 -15>
     1=<0 0 72>
     2=&FN_FrontEndVector3Modifier(Gameplay=<0 0 150>, Front End=<0 0 110>)
     3=<0 0 0>
     4=<0 0 20>
     5=&MakeCustomVectorFromBool(
      Bool=&SetBoolByIntComparison(A=$Engine.QualityLevel, B=0, Comparison Type=::NewEnumerator4)
      True Vector=<0 0 110>
      False Vector=<0 0 70>
     )
     6=<0 0 0>
     7=<0 0 3>
     Index=&SpawnGroup()
     Num=::NewEnumerator6
    )
    Sprite Rotation Mode=::NewEnumerator2
    Sprite Size=&MergeHelper_Vector2Switch(
     0=&RandomRangeVector2D(Minimum=<26 150>, Maximum=<130 300>)
     1=&Vector2DFromFloat(Value=&RandomRangeFloat(Minimum=200, Maximum=300))
     2=&Vector2DFromFloat(Value=&RandomRangeFloat(Minimum=200, Maximum=300))
     3=&Vector2DFromFloat(Value=&RandomRangeFloat(Minimum=80, Maximum=190))
     4=&Vector2DFromFloat(Value=&RandomRangeFloat(Minimum=80, Maximum=190))
     5=&FN_FrontEndVector2Modifier(Front End:Vector2f=<213.3333 266.6667>, Gameplay:Vector2f=<266.6667 333.3333>)
     6=<0 0>
     7=<200 200>
     Index=&SpawnGroup()
     Num=::NewEnumerator6
    )
    Sprite Size Mode=::NewEnumerator4
    UsePositionOffset=true
    Write Material Random=true
   )
   /SetVariables(Particles.RaySide:NiagaraInt32=&RandomRangeInt(Minimum=0, Maximum=1)) {isSetParameters=true}
   &SubUVAnimation(
    Evaluation Type=::NewEnumerator0
    Output Alternate Renderer Binding=true
    "Sprite Renderer "=DataInterface() {propertyValues {SpriteRenderer=&self:Ambient_Condensed_0.NiagaraSpriteRendererProperties_2}}
    SubUV Animation Mode:ENiagara_SubUVLookupModeV2=::NewEnumerator2
    UseEndFrame=false
    UseStartFrame=false
   )
   &ShapeLocation(
    Hemisphere Distribution=<1 0.55>
    MaskSpawnGroup=false
    Non Uniform Scale=<1 1 0.35>
    Offset=&MaskVector(
     Mask Source=0
     Mask Parameter=?
     Mask=0
     Masked Vector=&MaskVector(Mask Source=1, Mask Parameter=$Particles.RaySide, Mask=0, Masked Vector=<60 -40 170>, Passthrough Vector=<-60 -40 170>)
     Passthrough Vector=<0 0 0>
    )
    Offset Mode=::NewEnumerator0
    Override Seed=false
    Sphere Radius=&MaskFloat(Mask Source=0, Mask Parameter=?, Mask=0, Masked Float=10, Passthrough Float=0)
    UseOwnerScale=true
   ) {module=ShapeLocation001}
   &ShapeLocation(
    Cylinder Height=&MaskFloat(
     Mask Source=0
     Mask Parameter=?
     Mask=1
     Masked Float=&FN_FrontEndFloatModifier(Gameplay=50, Front End=35)
     Passthrough Float=0
    )
    Cylinder Radius=&MaskFloat(
     Mask Source=0
     Mask Parameter=?
     Mask=1
     Masked Float=&FN_FrontEndFloatModifier(Gameplay=30, Front End=40)
     Passthrough Float=0
    )
    MaskSpawnGroup=false
    Non Uniform Scale=<1.5 1.25 1>
    Override Seed=false
    Shape Primitive=::NewEnumerator1
    Surface Only=false
    UseOwnerScale=true
   )
   &ShapeLocation(
    Cylinder Height=&MaskFloat(
     Mask Source=0
     Mask Parameter=?
     Mask=2
     Masked Float=&FN_FrontEndFloatModifier(Gameplay=85, Front End=35)
     Passthrough Float=0
    )
    Cylinder Radius=&MaskFloat(
     Mask Source=0
     Mask Parameter=?
     Mask=2
     Masked Float=&FN_FrontEndFloatModifier(Gameplay=30, Front End=40)
     Passthrough Float=0
    )
    MaskSpawnGroup=false
    Non Uniform Scale=<1.5 1.25 1>
    Override Seed=false
    Shape Primitive=::NewEnumerator1
    Surface Only=false
    UseOwnerScale=true
   ) {module=ShapeLocation002}
   &ShapeLocation(
    MaskSpawnGroup=false
    Non Uniform Scale=<1.25 1 1>
    Override Seed=false
    Ring Radius=&MaskFloat(Mask Source=0, Mask Parameter=?, Mask=3, Masked Float=30, Passthrough Float=0)
    Shape Primitive=::NewEnumerator4
    UseOwnerScale=true
   ) {module=ShapeLocation003}
   &ShapeLocation(
    Cylinder Height=&MaskFloat(Mask Source=0, Mask Parameter=?, Mask=4, Masked Float=25, Passthrough Float=0)
    Cylinder Radius=&MaskFloat(Mask Source=0, Mask Parameter=?, Mask=4, Masked Float=20, Passthrough Float=0)
    MaskSpawnGroup=false
    Non Uniform Scale=<1.25 1 1>
    Override Seed=false
    Shape Primitive=::NewEnumerator1
    Surface Only=false
    UseOwnerScale=true
   ) {module=ShapeLocation004}
   &AddVelocity(
    Module.Use Falloff Distance=false
    Origin Offset=<0 0 -50>
    Override Seed=false
    Velocity Mode=::NewEnumerator1
    Velocity Speed=&MaskFloat(Mask Source=0, Mask Parameter=?, Mask=0, Masked Float=0.01, Passthrough Float=0)
   )
   &AddVelocity(Velocity=&MergeHelper_VectorSwitch(
     0=&RandomRangeVector(
      Minimum=&MaskVector(Mask Source=1, Mask Parameter=$Particles.RaySide, Mask=0, Masked Vector=<-5 1 0>, Passthrough Vector=<5 1 0>)
      Maximum=<0.1 5 -5>
     )
     1=<0 0 0>
     2=<0 0 0>
     3=<0 0 3>
     4=<0 0 12>
     5=<0 0 0>
     6=<0 0 0>
     7=<0 0 0>
     Index=&SpawnGroup()
     Num=::NewEnumerator6
    )) {module=AddVelocity001}
   &CameraOffset(Camera Offset Amount=&MergeHelper_FloatSwitch(
     0=0
     1=-50
     2=-50
     3=&RandomRangeFloat(Minimum=-2, Maximum=25)
     4=&RandomRangeFloat(Minimum=-8, Maximum=15)
     5=-3
     Index=&SpawnGroup()
     Num=::NewEnumerator4
    ))
   /SetVariables(Particles.Glow Orientation:Vector3f=<0 0 1>) {isSetParameters=true}
  ]
  particleUpdateScript [
   &ParticleState()
   &ScaleSpriteSize(
    Scale Factor=&MergeHelper_Vector2Switch(
     0=&Vector2DFromCurve(
      Vector2Curve=DataInterface() {propertyValues {
        XCurve {
         defaultValue=0
         keys [
          {interpMode=RCIM_Cubic, tangentMode=RCTM_User, time=0, value=0}
          {arriveTangentWeight=0.1667, interpMode=RCIM_Cubic, tangentMode=RCTM_User, time=0.5, value=1}
          {arriveTangent=-3.3018, interpMode=RCIM_Cubic, leaveTangent=-3.3018, tangentMode=RCTM_User, time=0.8652, value=0.4388}
          <at 1 0>
         ]
        }
        YCurve {
         defaultValue=0
         keys [
          {interpMode=RCIM_Cubic, tangentMode=RCTM_User, time=0, value=0}
          {arriveTangent=2.4744, interpMode=RCIM_Cubic, leaveTangent=2.4744, tangentMode=RCTM_User, time=0.209, value=0.758}
          {arriveTangentWeight=0.1667, interpMode=RCIM_Cubic, tangentMode=RCTM_User, time=0.5, value=1}
          <at 1 0>
         ]
        }
       }}
      ScaleCurve=<1 1>
     )
     1=&Vector2DFromFloat(Value=&FloatFromCurve(FloatCurve=DataInterface() {propertyValues {Curve {keys [{tangentMode=RCTM_User, time=0, value=1}, {tangentMode=RCTM_User, time=1, value=1.2}]}}}))
     2=&Vector2DFromFloat(Value=&FloatFromCurve(FloatCurve=DataInterface() {propertyValues {Curve {keys [{tangentMode=RCTM_User, time=0, value=1}, {tangentMode=RCTM_User, time=1, value=1.2}]}}}))
     3=<1 1>
     4=&Vector2DFromFloat(Value=&FloatFromCurve(FloatCurve=DataInterface() {propertyValues {Curve {
          defaultValue=0
          keys [
           {interpMode=RCIM_Cubic, tangentMode=RCTM_User, time=0, value=0}
           {interpMode=RCIM_Cubic, tangentMode=RCTM_User, time=0.29, value=1}
           {arriveTangent=-1.4085, interpMode=RCIM_Cubic, leaveTangent=-1.4085, time=0.6493, value=0.8525}
           {tangentMode=RCTM_User, time=1, value=0}
          ]
         }}}))
     5=<1 1>
     6=<1 1>
     7=<1 1>
     Index=&SpawnGroup()
     Num=::NewEnumerator6
    )
    Scale Sprite Size Mode=::NewEnumerator2
   )
   &SubUVAnimation(
    End Frame Range Override=40
    Output Alternate Renderer Binding=true
    "Sprite Renderer "=DataInterface() {propertyValues {SpriteRenderer=&self:Ambient_Condensed_0.NiagaraSpriteRendererProperties_1}}
    SubUV Lookup Index Scale=0.6
    UseEndFrame=true
    UseStartFrame=true
   ) {module=SubUVAnimation001}
   &SubUVAnimation(
    End Frame Range Override=40
    Output Alternate Renderer Binding=true
    "Sprite Renderer "=DataInterface() {propertyValues {SpriteRenderer=&self:Ambient_Condensed_0.NiagaraSpriteRendererProperties_3}}
    SubUV Lookup Index Scale=0.6
    UseEndFrame=true
    UseStartFrame=true
   ) {module=SubUVAnimation002}
   &SubUVAnimation(
    Output Alternate Renderer Binding=true
    "Sprite Renderer "=DataInterface() {propertyValues {SpriteRenderer=&self:Ambient_Condensed_0.NiagaraSpriteRendererProperties_0}}
    UseEndFrame=false
    UseStartFrame=false
   ) {module=SubUVAnimation003}
   &SubUVAnimation(
    Output Alternate Renderer Binding=true
    "Sprite Renderer "=DataInterface() {propertyValues {SpriteRenderer=&self:Ambient_Condensed_0.NiagaraSpriteRendererProperties_4}}
    UseEndFrame=false
    UseStartFrame=false
   ) {module=SubUVAnimation004}
   &ScaleMeshSize(Scale Factor=&Multiply_VectorByFloat(
     Vector=&Multiply_VectorByFloat(Vector=$Engine.Owner.Scale, Float=&FN_FrontEndFloatModifier(Gameplay=0.9, Front End=0.8))
     Float=&FloatFromCurve(
      FloatCurve=DataInterface() {propertyValues {Curve {keys [
          {arriveTangent=0.0893, interpMode=RCIM_Cubic, leaveTangent=0.0893, tangentMode=RCTM_User, time=0, value=0.55}
          {arriveTangent=0.0158, interpMode=RCIM_Cubic, leaveTangent=0.0158, tangentMode=RCTM_User, time=1, value=1}
         ]}}}
      Scale Curve=&Multiply_Float(
       A=4
       B=&FloatFromCurve(
        FloatCurve=DataInterface() {propertyValues {Curve {keys [{interpMode=RCIM_Cubic, tangentMode=RCTM_User, time=0, value=0.4}, {interpMode=RCIM_Cubic, tangentMode=RCTM_User, time=0.2095, value=1}]}}}
        CurveIndex=$Emitter.Age
       )
      )
     )
    )) {module=ScaleMeshSize001}
   &UpdateMeshOrientation(
    Rotation Rate=&SelectFloatFromArray(
     Array Sampling Mode=::NewEnumerator0
     Direct Array Index=?
     Float Selection Array=DataInterface() {propertyValues {FloatData [-0.3,0.3,0.2,-0.2]}}
     Random Seed=0
    )
    Rotation Vector=<0 0 1>
   )
   /SetVariables(Particles.Glow Dyn:Vector4f=<1.7 1 1 1>) {isSetParameters=true}
   &GravityForce(Gravity=&MaskVector(
     Mask Source=0
     Mask Parameter=?
     Mask=4
     Masked Vector=&RandomRangeVector(Minimum=<0 0 8>, Maximum=<0 0 20>)
     Passthrough Vector=<0 0 0>
    ))
   &Drag(
    Drag=&MaskFloat(Mask Source=0, Mask Parameter=?, Mask=4, Masked Float=0.5, Passthrough Float=0)
    Rotational Drag:NiagaraFloat=&MaskFloat(Mask Source=0, Mask Parameter=?, Mask=4, Masked Float=1, Passthrough Float=0)
    Use Linear Drag=true
    Use Rotational Drag=true
   )
   &ScaleColor(
    Scale Alpha=&MergeHelper_FloatSwitch(
     0=&FloatFromCurve(
      FloatCurve=DataInterface() {propertyValues {Curve {
         defaultValue=0
         keys [
          {interpMode=RCIM_Cubic, tangentMode=RCTM_User, time=0, value=0}
          {interpMode=RCIM_Cubic, tangentMode=RCTM_User, time=0.5, value=1}
          {tangentMode=RCTM_User, time=1, value=0}
         ]
        }}}
      Scale Curve=&Multiply_Float(
       A=&MakeCustomFloatFromBool(
        Bool=&SetBoolByFloatComparison(
         A=&FN_PickFloatDependOnPlatform(
          TargetPlatform=DataInterface() {propertyValues {Platforms {
             qualityLevelMask=-30
             deviceProfileStates [
              {profileName=Windows, qualityLevelMask=0, setQualityLevelMask=2}
              {profileName=Mobile, qualityLevelMask=0, setQualityLevelMask=4294967295}
              {profileName=Mac, qualityLevelMask=0, setQualityLevelMask=2}
              {profileName=Linux, qualityLevelMask=0, setQualityLevelMask=2}
             ]
            }}}
          TargetPlatformActive=1
          TargetPlatformNoActive=0
         )
         B=0
        )
        True Float=0.5
        False Float=1
       )
       B=&Multiply_Float(
        A=&MakeCustomFloatFromBool(
         Bool=&SetBoolByIntComparison(A=$Engine.QualityLevel, B=0, Comparison Type=::NewEnumerator4)
         True Float=0.65
         False Float=1
        )
        B=&Multiply_Float(
         A=&RandomRangeFloat(Minimum=0.23, Maximum=1.1)
         B=&FNTimeOfDayScalar(Night Scale=0.4, Morning Scale=0.75, Evening Scale=0.75)
        )
       )
      )
     )
     1=&FloatFromCurve(
      FloatCurve=DataInterface() {propertyValues {Curve {keys [
          {interpMode=RCIM_Cubic, tangentMode=RCTM_User, time=0, value=0}
          {interpMode=RCIM_Cubic, tangentMode=RCTM_User, time=0.2766, value=0.3}
          {tangentMode=RCTM_User, time=1, value=0}
         ]}}}
      Scale Curve=&Multiply_Float(
       A=&MakeCustomFloatFromBool(
        Bool=&SetBoolByFloatComparison(
         A=&FN_PickFloatDependOnPlatform(
          TargetPlatform=DataInterface() {propertyValues {Platforms {
             qualityLevelMask=-30
             deviceProfileStates [
              {profileName=Windows, qualityLevelMask=0, setQualityLevelMask=2}
              {profileName=Mobile, qualityLevelMask=0, setQualityLevelMask=4294967295}
              {profileName=Mac, qualityLevelMask=0, setQualityLevelMask=2}
              {profileName=Linux, qualityLevelMask=0, setQualityLevelMask=2}
             ]
            }}}
          TargetPlatformActive=1
          TargetPlatformNoActive=0
         )
         B=0
        )
        True Float=0.5
        False Float=1
       )
       B=&Multiply_Float(
        A=&MakeCustomFloatFromBool(
         Bool=&SetBoolByIntComparison(A=$Engine.QualityLevel, B=0, Comparison Type=::NewEnumerator4)
         True Float=0.65
         False Float=1
        )
        B=&Multiply_Float(
         A=&FN_FrontEndFloatModifier(Gameplay=0.22, Front End=0.12)
         B=&FNTimeOfDayScalar(Night Scale=0.5, Morning Scale=0.75, Evening Scale=0.75)
        )
       )
      )
     )
     2=&FloatFromCurve(
      FloatCurve=DataInterface() {propertyValues {Curve {keys [
          {interpMode=RCIM_Cubic, tangentMode=RCTM_User, time=0, value=0}
          {interpMode=RCIM_Cubic, tangentMode=RCTM_User, time=0.2766, value=0.1}
          {tangentMode=RCTM_User, time=1, value=0}
         ]}}}
      Scale Curve=&Multiply_Float(
       A=&MakeCustomFloatFromBool(
        Bool=&SetBoolByFloatComparison(
         A=&FN_PickFloatDependOnPlatform(
          TargetPlatform=DataInterface() {propertyValues {Platforms {
             qualityLevelMask=-30
             deviceProfileStates [
              {profileName=Windows, qualityLevelMask=0, setQualityLevelMask=2}
              {profileName=Mobile, qualityLevelMask=0, setQualityLevelMask=4294967295}
              {profileName=Mac, qualityLevelMask=0, setQualityLevelMask=2}
              {profileName=Linux, qualityLevelMask=0, setQualityLevelMask=2}
             ]
            }}}
          TargetPlatformActive=1
          TargetPlatformNoActive=0
         )
         B=0
        )
        True Float=0.5
        False Float=1
       )
       B=&Multiply_Float(
        A=&MakeCustomFloatFromBool(
         Bool=&SetBoolByIntComparison(A=$Engine.QualityLevel, B=0, Comparison Type=::NewEnumerator4)
         True Float=0.65
         False Float=1
        )
        B=&Multiply_Float(
         A=&FN_FrontEndFloatModifier(Gameplay=0.3, Front End=0.12)
         B=&FNTimeOfDayScalar(Night Scale=0.5, Morning Scale=0.75, Evening Scale=0.75)
        )
       )
      )
     )
     3=&FloatFromCurve(
      FloatCurve=DataInterface() {propertyValues {Curve {
         defaultValue=1
         keys [
          {interpMode=RCIM_Cubic, tangentMode=RCTM_User, time=0, value=0}
          {interpMode=RCIM_Cubic, tangentMode=RCTM_User, time=0.2807, value=1.0588}
          {tangentMode=RCTM_User, time=1, value=0}
         ]
        }}}
      Scale Curve=&Multiply_Float(
       A=&MakeCustomFloatFromBool(
        Bool=&SetBoolByFloatComparison(
         A=&FN_PickFloatDependOnPlatform(
          TargetPlatform=DataInterface() {propertyValues {Platforms {
             qualityLevelMask=-29
             deviceProfileStates [
              {profileName=Windows, qualityLevelMask=0, setQualityLevelMask=3}
              {profileName=Linux, qualityLevelMask=0, setQualityLevelMask=3}
              {profileName=Mac, qualityLevelMask=0, setQualityLevelMask=3}
              {profileName=IOS, qualityLevelMask=0, setQualityLevelMask=4294967295}
             ]
            }}}
          TargetPlatformActive=1
          TargetPlatformNoActive=0
         )
         B=0
        )
        True Float=0.5
        False Float=1
       )
       B=&Multiply_Float(
        A=&MakeCustomFloatFromBool(
         Bool=&SetBoolByIntComparison(A=$Engine.QualityLevel, B=0, Comparison Type=::NewEnumerator4)
         True Float=0.65
         False Float=1
        )
        B=&Multiply_Float(
         A=&FN_FrontEndFloatModifier(Gameplay=0.65, Front End=0.3)
         B=&FNTimeOfDayScalar(Night Scale=0.6, Morning Scale=0.8, Evening Scale=0.8)
        )
       )
      )
     )
     4=&FloatFromCurve(
      FloatCurve=DataInterface() {propertyValues {Curve {
         defaultValue=0
         keys [
          {arriveTangent=13.0062, interpMode=RCIM_Cubic, leaveTangent=13.0062, tangentMode=RCTM_User, time=0, value=0}
          {interpMode=RCIM_Cubic, tangentMode=RCTM_User, time=0.1852, value=1}
          {arriveTangent=-0.7697, interpMode=RCIM_Cubic, leaveTangent=-0.7697, tangentMode=RCTM_User, time=0.7051, value=0.8679}
          {arriveTangent=-1.9584, interpMode=RCIM_Cubic, leaveTangent=-1.9584, tangentMode=RCTM_User, time=1, value=0}
         ]
        }}}
      Scale Curve=&Multiply_Float(
       A=&MakeCustomFloatFromBool(
        Bool=&SetBoolByFloatComparison(
         A=&FN_PickFloatDependOnPlatform(
          TargetPlatform=DataInterface() {propertyValues {Platforms {
             qualityLevelMask=-30
             deviceProfileStates [
              {profileName=Windows, qualityLevelMask=0, setQualityLevelMask=2}
              {profileName=Mobile, qualityLevelMask=0, setQualityLevelMask=4294967295}
              {profileName=Mac, qualityLevelMask=0, setQualityLevelMask=2}
              {profileName=Linux, qualityLevelMask=0, setQualityLevelMask=2}
             ]
            }}}
          TargetPlatformActive=1
          TargetPlatformNoActive=0
         )
         B=0
        )
        True Float=0.5
        False Float=1
       )
       B=&Multiply_Float(
        A=&MakeCustomFloatFromBool(
         Bool=&SetBoolByIntComparison(A=$Engine.QualityLevel, B=0, Comparison Type=::NewEnumerator4)
         True Float=0.33
         False Float=1
        )
        B=&Multiply_Float(
         A=&FN_FrontEndFloatModifier(
          Gameplay=&RandomRangeFloat(Minimum=0.6, Maximum=0.83)
          Front End=&FloatFromCurve(FloatCurve=DataInterface() {propertyValues {Curve {
              defaultValue=0.3
              keys [
               {interpMode=RCIM_Cubic, tangentMode=RCTM_User, time=0, value=0.4}
               {arriveTangent=-1.1537, interpMode=RCIM_Cubic, leaveTangent=-1.1537, tangentMode=RCTM_User, time=1, value=0}
              ]
             }}})
         )
         B=&FNTimeOfDayScalar(Night Scale=0.6, Morning Scale=0.8, Evening Scale=0.8)
        )
       )
      )
     )
     5=&FloatFromCurve(
      FloatCurve=DataInterface() {propertyValues {Curve {keys [<at 0 0>,<at 0.15 1>,<at 0.85 1>,<at 1 0>]}}}
      Scale Curve=&Multiply_Float(
       A=&MakeCustomFloatFromBool(
        Bool=&SetBoolByFloatComparison(
         A=&FN_PickFloatDependOnPlatform(
          TargetPlatform=DataInterface() {propertyValues {Platforms {
             qualityLevelMask=-29
             deviceProfileStates [
              {profileName=Windows, qualityLevelMask=0, setQualityLevelMask=3}
              {profileName=Linux, qualityLevelMask=0, setQualityLevelMask=3}
              {profileName=Mac, qualityLevelMask=0, setQualityLevelMask=3}
              {profileName=IOS, qualityLevelMask=0, setQualityLevelMask=4294967295}
             ]
            }}}
          TargetPlatformActive=1
          TargetPlatformNoActive=0
         )
         B=0
        )
        True Float=0.2
        False Float=1
       )
       B=&FN_FrontEndFloatModifier(Gameplay=1, Front End=0.6)
      )
     )
     6=&FloatFromCurve(FloatCurve=DataInterface() {propertyValues {Curve {
         defaultValue=1
         keys [
          {interpMode=RCIM_Cubic, tangentMode=RCTM_User, time=0, value=0}
          {interpMode=RCIM_Cubic, tangentMode=RCTM_User, time=0.5, value=1}
          {tangentMode=RCTM_User, time=1, value=0}
         ]
        }}})
     7=&FloatFromCurve(
      FloatCurve=DataInterface() {propertyValues {Curve {
         defaultValue=1
         keys [
          {interpMode=RCIM_Cubic, tangentMode=RCTM_User, time=0, value=0.15}
          {interpMode=RCIM_Cubic, tangentMode=RCTM_User, time=0.2002, value=1}
          {arriveTangent=-1.0624, interpMode=RCIM_Cubic, leaveTangent=-1.0624, time=0.8198, value=1}
          {tangentMode=RCTM_User, time=1, value=0.1503}
         ]
        }}}
      Scale Curve=0.7
     )
     Index=&SpawnGroup()
     Num=::NewEnumerator6
    )
    Scale RGB=&MergeHelper_VectorSwitch(
     0=<1 1 1>
     1=&Multiply_VectorByFloat(Vector=$System.LightColor, Float=&RandomRangeFloat(Minimum=0.5, Maximum=1))
     2=&Multiply_VectorByFloat(Vector=$System.LightColor, Float=&RandomRangeFloat(Minimum=0.5, Maximum=1))
     3=$System.LightColor
     4=$System.LightColor
     5=$System.LightColor
     6=$System.LightColor
     7=$System.LightColor
     Index=&SpawnGroup()
     Num=::NewEnumerator6
    )
    ScaleA=true
    ScaleRGB=true
   )
   &CurlNoiseForce(
    Bias Noise=true
    Bias Noise Field:Vector3f=<0 1 0>
    Noise Strength=&MaskFloat(Mask Source=0, Mask Parameter=?, Mask=0, Masked Float=10, Passthrough Float=0)
    Pan Noise=false
   )
   &SolveForcesAndVelocity(Clamp Velocity=false, Limit Acceleration=false)
   &BlendSpace(Rotation Contribution=0)
  ]
  rendererClasses [&NiagaraSpriteRendererProperties, &NiagaraMeshRendererProperties]
  renderers [
   {rendererClass=&NiagaraSpriteRendererProperties}
   {rendererClass=&NiagaraSpriteRendererProperties, rendererIndex=1}
   {rendererClass=&NiagaraSpriteRendererProperties, rendererIndex=2}
   {rendererClass=&NiagaraSpriteRendererProperties, rendererIndex=3}
   {rendererClass=&NiagaraSpriteRendererProperties, rendererIndex=4}
   {rendererClass=&NiagaraSpriteRendererProperties, rendererIndex=5}
   {rendererClass=&NiagaraMeshRendererProperties, rendererIndex=6}
   {rendererClass=&NiagaraSpriteRendererProperties, rendererIndex=7}
  ]
 }
 Back_Light {rendererClasses [], renderers []}
 Embers {
  bEnabled=false
  emitterUpdateScript [
   &EmitterState()
   &SpawnRate(
    SpawnRate=&Multiply_Float(
     A=&Multiply_Float(A=25, B=&MakeFloatFromInt(Int=$System.PyroStage1))
     B=$User.Pyrotechnics
    )
    Use Spawn Probability=false
   )
   &SpawnRate(
    SpawnGroup=1
    SpawnRate=&Multiply_Float(
     A=&Multiply_Float(A=50, B=&MakeFloatFromInt(Int=$System.PyroStage2))
     B=$User.Pyrotechnics
    )
    Use Spawn Probability=false
   ) {module=SpawnRate001}
   &SpawnRate(
    SpawnGroup=2
    SpawnRate=&Multiply_Float(
     A=&Multiply_Float(A=50, B=&MakeFloatFromInt(Int=$System.PyroStage3))
     B=$User.Pyrotechnics
    )
    Use Spawn Probability=false
   ) {module=SpawnRate003}
   &SpawnRate(
    SpawnGroup=3
    SpawnRate=&Multiply_Float(
     A=&Multiply_Float(A=50, B=&MakeFloatFromInt(Int=$System.PyroStage4))
     B=$User.Pyrotechnics
    )
    Use Spawn Probability=false
   ) {module=SpawnRate004}
   &SpawnRate(
    SpawnGroup=4
    SpawnRate=&Multiply_Float(
     A=&Multiply_Float(A=30, B=&MakeFloatFromInt(Int=$System.PyroStage5))
     B=$User.Pyrotechnics
    )
    Use Spawn Probability=false
   ) {module=SpawnRate005}
  ]
  particleSpawnScript [
   &PartitionParticles(
    A:NiagaraFloat=&RandomRangeFloat(Minimum=0, Maximum=1)
    B=0.7
    Partition Mode:ENiagara_PartitionMode_v2=::NewEnumerator4
   )
   /SetVariables(Particles.VisibilityTag=$Particles.PartitionParticles.Partition) {isSetParameters=true}
   &InitializeParticle(
    Lifetime=&MaskFloat(
     Mask Source=1
     Mask Parameter=$Particles.PartitionParticles.Partition
     Mask=0
     Masked Float=&UniformRangedFloat(Minimum=0.5, Maximum=1)
     Passthrough Float=5
    )
    Mass Mode=::NewEnumerator1
    Position=&SimulationPosition()
    Position Mode=::NewEnumerator1
    Position Offset=<0 0 10>
    Sprite Size=&Vector2DFromFloat(Value=&FloatFromCurve(
      FloatCurve=DataInterface() {propertyValues {Curve {keys [<at 0 1.2>,<at 1 0.5>]}}}
      CurveIndex=$Emitter.NormalizedLoopAge
      Scale Curve=&RandomRangeFloat(Minimum=1, Maximum=3)
     ))
    Sprite Size Mode=::NewEnumerator4
    UsePositionOffset=false
    Write Material Random=true
   ) {module=InitializeParticle001}
   &ShapeLocation(
    Cylinder Height=1
    Cylinder Radius=0.1
    MaskSpawnGroup=false
    Offset=&Multiply_Vector(
     A=/Niagara/DynamicInputs/Helpers/MaskPosition(
      Mask Source:NiagaraInt32=0
      Mask Parameter:NiagaraInt32=?
      Mask:NiagaraInt32=0
      Masked Position:NiagaraPosition=<1 1 1>
      Passthrough Position:NiagaraPosition=&MaskPosition(
       Mask Source:NiagaraInt32=1
       Mask Parameter:NiagaraInt32=0
       Mask:NiagaraInt32=&MakeIntFromFloat(Float=&UniformAOrBFloat(A=0, Evaluation Type=::NewEnumerator0, Recalculate Random Each Loop=true, Override Seed=false))
       Masked Position:NiagaraPosition=<1 1 1>
       Passthrough Position:NiagaraPosition=<-1 1 1>
      )
     )
     B=&SelectVectorFromArray(Direct Array Index=$Emitter.SpawnGroup, Vector Selection Array=$System.Fire Array)
    )
    Offset Mode=::NewEnumerator0
    Override Seed=false
    Shape Primitive=::NewEnumerator1
    Surface Only=false
    UseOwnerScale=true
   )
   &AddVelocity(
    Velocity=&RandomRangeVector(Minimum=<20 20 25>, Maximum=<-20 -20 50>)
    Velocity Speed Scale=&FloatFromCurve(
     FloatCurve=DataInterface() {propertyValues {Curve {keys [<at 0 1 RCIM_Constant>,<at 0.2 2 RCIM_Constant>,<at 0.4 3 RCIM_Constant>,<at 0.6 4 RCIM_Constant>,<at 0.8 5 RCIM_Constant>]}}}
     CurveIndex=&RemapRange(
      Input Value To Remap=&MakeFloatFromInt(Int=$Emitter.SpawnGroup)
      Input Min=0
      Input Max=4
      Output Min=0
      Output Max=1
     )
     Scale Curve=2
    )
   )
   /SetVariables(
    Particles.GlowCol:LinearColor=<rgba 10 1.1758 0 1>
    Particles.GlowFacing:Vector3f=<0 0 1>
    Particles.GlowPos:NiagaraPosition=&Multiply_Vector(
     A=&MaskPosition(
      Mask Source:NiagaraInt32=0
      Mask Parameter:NiagaraInt32=?
      Mask:NiagaraInt32=0
      Masked Position:NiagaraPosition=<1 1 1>
      Passthrough Position:NiagaraPosition=&MaskPosition(
       Mask Source:NiagaraInt32=1
       Mask Parameter:NiagaraInt32=0
       Mask:NiagaraInt32=&MakeIntFromFloat(Float=&UniformAOrBFloat(A=0, Evaluation Type=::NewEnumerator0, Recalculate Random Each Loop=true, Override Seed=false))
       Masked Position:NiagaraPosition=<1 1 1>
       Passthrough Position:NiagaraPosition=<-1 1 1>
      )
     )
     B=&SelectVectorFromArray(Direct Array Index=$Emitter.SpawnGroup, Vector Selection Array=$System.Fire Array)
    )
   ) {isSetParameters=true}
  ]
  particleUpdateScript [
   &ParticleState()
   &ScaleSpriteSize(Uniform Curve Sprite Scale=DataInterface() {propertyValues {Curve {keys [
        {arriveTangent=0.178, interpMode=RCIM_Cubic, leaveTangent=0.178, tangentMode=RCTM_User, time=0, value=0}
        {arriveTangent=-2.5249, interpMode=RCIM_Cubic, leaveTangent=-2.5249, tangentMode=RCTM_User, time=0.1408, value=0.9847}
        {arriveTangent=-0.0405, interpMode=RCIM_Cubic, leaveTangent=-0.0405, tangentMode=RCTM_User, time=1, value=0}
       ]}}})
   &GravityForce(Gravity=<0 0 350>)
   &Drag(Drag=2, Use Linear Drag=true, Use Rotational Drag=false)
   /Niagara/Modules/Update/Forces/CurlNoiseForce(Noise Frequency=0.1, Noise Strength=50, Pan Noise=true, Pan Noise Field=<0 0 1>)
   &SolveForcesAndVelocity(Clamp Velocity=true, Limit Acceleration=false, Speed Limit=350)
   &ScaleSpriteSizeBySpeed(
    Max Scale Factor=<1 4>
    Min Scale Factor=<0.5 0.5>
    Sample Scale Factor By Curve=true
    Scale Factor Curve=DataInterface() {propertyValues {Curve {keys [
        <at -0.0077 0.0765 RCIM_Cubic>
        {arriveTangent=1.2463, interpMode=RCIM_Cubic, leaveTangent=1.2463, tangentMode=RCTM_User, time=0.9985, value=0.9549}
       ]}}}
    Velocity Threshold=500
   )
   &ScaleColor(
    Scale Alpha=&FloatFromCurve(FloatCurve=DataInterface() {propertyValues {Curve {keys [<at 0 1>,<at 1 0>]}}})
    Scale RGB=&MakeVectorFromLinearColorRGB(Color=&ColorFromCurve(ColorCurve=DataInterface() {propertyValues {
        AlphaCurve {keys [<at 0 1 RCIM_Cubic>,<at 0.7107 1>]}
        BlueCurve {keys [<at 0 14.6701 RCIM_Cubic>,<at 0.2499 1.5104>,<at 0.7114 0.1753 RCIM_Cubic>]}
        GreenCurve {keys [<at 0 36.7481 RCIM_Cubic>,<at 0.2499 5.293>,<at 0.7114 0.3243 RCIM_Cubic>]}
        RedCurve {keys [<at 0 50 RCIM_Cubic>,<at 0.2499 10>,<at 0.7114 1 RCIM_Cubic>]}
       }}))
    ScaleA=true
    ScaleRGB=true
   )
   &CameraOffset(Camera Offset Amount=2)
   /SetVariables(
    Particles.GlowCol:LinearColor=&MultiplyLinearColor(
     Linear Color=$Particles.Initial.GlowCol
     Scale Factor=&MakeVector4FromVectorAndFloat(
      "Vector (RGB)"=&VectorFromFloat(Value=&FloatFromCurve(FloatCurve=DataInterface() {propertyValues {Curve {keys [<at 0.0119 0>,<at 0.1 1>]}}}))
      "Float (Alpha)"=$User.Pyrotechnics
     )
    )
    Particles.GlowDyn:Vector4f=<3 1 1 1>
    Particles.Vector 2D:Vector2f=<50 50>
   ) {isSetParameters=true}
  ]
  rendererClasses [&NiagaraSpriteRendererProperties]
  renderers [
   {rendererClass=&NiagaraSpriteRendererProperties}
   {rendererClass=&NiagaraSpriteRendererProperties, rendererIndex=1}
  ]
 }
 Flashing_Lights {
  emitterUpdateScript [
   &EmitterState(Delay First Loop Only=true, Life Cycle Mode=::NewEnumerator1, Loop Delay=0.9, Loop Duration=0.9023, UseLoopDelay=true)
   /Niagara/Modules/Spawn/Location/SpawnParticlesInGrid(X Count=5, Y Count=0, Z Count=3)
   &SpawnParticlesInGrid(Spawn Group=1, X Count=5, Y Count=0, Z Count=3) {module=SpawnParticlesInGrid001}
  ]
  particleSpawnScript [
   &InitializeParticle(
    Color=<rgba 0.8482 0.8345 1 1>
    Lifetime=&RandomRangeFloat(Minimum=0.3, Maximum=0.6)
    Sprite Size=&RandomRangeVector2D(Minimum=<26 150>, Maximum=<30 300>)
    Sprite Size Mode=::NewEnumerator4
    UsePositionOffset=false
    Write Material Random=true
   )
   /Niagara/Modules/Spawn/Location/V2/GridLocation(
    Module.Jitter Inside Grid Cells=false
    Module.Override Random Seed=false
    Offset:Vector3f=&MaskVector(Mask Source=0, Mask Parameter=?, Mask=0, Masked Vector=<6 -50 18>, Passthrough Vector=<-6 -50 18>)
    XYZ Dimensions:Vector3f=<10 1 10>
   )
   &SubUVAnimation(
    Evaluation Type=::NewEnumerator0
    Output Alternate Renderer Binding=true
    "Sprite Renderer "=DataInterface() {propertyValues {SpriteRenderer=&self:Flashing_Lights_0.NiagaraSpriteRendererProperties_2}}
    SubUV Animation Mode:ENiagara_SubUVLookupModeV2=::NewEnumerator2
    UseEndFrame=false
    UseStartFrame=false
   )
   &AddVelocity(
    Velocity=&RandomRangeVector(
     Minimum=&MaskVector(Mask Source=1, Mask Parameter=$Emitter.SpawnGroup, Mask=0, Masked Vector=<-3 1 0>, Passthrough Vector=<3 1 0>)
     Maximum=<0.1 5 -5>
    )
    Velocity Speed Scale=0.01
   )
   &ApplyInitialForces()
   &KillParticles(Kill Particles=&SetBoolByFloatComparison(
     A=&UniformAOrBFloat(A=0, Distribution Weight=0.7, Evaluation Type=::NewEnumerator0, Recalculate Random Each Loop=true, Override Seed=false)
     B=0.5
    ))
   /SetVariables(Particles.LightSize:Vector2f=<20 20>) {isSetParameters=true}
  ]
  particleUpdateScript [
   &ParticleState()
   &ScaleSpriteSize(
    Scale Factor=&Vector2DFromCurve(
     Vector2Curve=DataInterface() {propertyValues {
       XCurve {
        defaultValue=0
        keys [
         {interpMode=RCIM_Cubic, tangentMode=RCTM_User, time=0, value=0}
         {arriveTangentWeight=0.1667, interpMode=RCIM_Cubic, tangentMode=RCTM_User, time=0.5, value=1}
         {arriveTangent=-3.3018, interpMode=RCIM_Cubic, leaveTangent=-3.3018, tangentMode=RCTM_User, time=0.8652, value=0.4388}
         <at 1 0>
        ]
       }
       YCurve {
        defaultValue=0
        keys [
         {interpMode=RCIM_Cubic, tangentMode=RCTM_User, time=0, value=0}
         {arriveTangent=2.4744, interpMode=RCIM_Cubic, leaveTangent=2.4744, tangentMode=RCTM_User, time=0.209, value=0.758}
         {arriveTangentWeight=0.1667, interpMode=RCIM_Cubic, tangentMode=RCTM_User, time=0.5, value=1}
         <at 1 0>
        ]
       }
      }}
     ScaleCurve=<1 1>
    )
    Scale Sprite Size Mode=::NewEnumerator2
   )
   /Niagara/Modules/Update/Forces/V2/CurlNoiseForce(Bias Noise=true, Bias Noise Field:Vector3f=<0 1 0>, Noise Frequency=0, Noise Strength=0, Pan Noise=false)
   &SolveForcesAndVelocity(Clamp Velocity=false, Limit Acceleration=false)
   &ScaleColor(
    Scale Alpha=&FloatFromCurve(
     FloatCurve=DataInterface() {propertyValues {Curve {
        defaultValue=0
        keys [
         {interpMode=RCIM_Cubic, tangentMode=RCTM_User, time=0, value=0}
         {interpMode=RCIM_Cubic, tangentMode=RCTM_User, time=0.1374, value=1}
         {tangentMode=RCTM_User, time=1, value=0}
        ]
       }}}
     Scale Curve=&Multiply_Float(
      A=&MakeCustomFloatFromBool(
       Bool=&SetBoolByFloatComparison(
        A=&FN_PickFloatDependOnPlatform(
         TargetPlatform=DataInterface() {propertyValues {Platforms {
            qualityLevelMask=-30
            deviceProfileStates [
             {profileName=Windows, qualityLevelMask=0, setQualityLevelMask=2}
             {profileName=Mobile, qualityLevelMask=0, setQualityLevelMask=4294967295}
             {profileName=Mac, qualityLevelMask=0, setQualityLevelMask=2}
             {profileName=Linux, qualityLevelMask=0, setQualityLevelMask=2}
            ]
           }}}
         TargetPlatformActive=1
         TargetPlatformNoActive=0
        )
        B=0
       )
       True Float=0.5
       False Float=1
      )
      B=&Multiply_Float(
       A=&MakeCustomFloatFromBool(
        Bool=&SetBoolByIntComparison(A=$Engine.QualityLevel, B=0, Comparison Type=::NewEnumerator4)
        True Float=0.65
        False Float=1
       )
       B=&Multiply_Float(
        A=&RandomRangeFloat(Minimum=0.23, Maximum=1.1)
        B=&FNTimeOfDayScalar(Night Scale=0.4, Morning Scale=0.75, Evening Scale=0.75)
       )
      )
     )
    )
    ScaleA=true
    ScaleRGB=true
   )
  ]
  rendererClasses [&NiagaraSpriteRendererProperties]
  renderers [
   {rendererClass=&NiagaraSpriteRendererProperties}
   {rendererClass=&NiagaraSpriteRendererProperties, rendererIndex=1}
   {rendererClass=&NiagaraSpriteRendererProperties, rendererIndex=2}
  ]
 }
 Front_Light {rendererClasses [], renderers []}
 Leg_Lights {rendererClasses [], renderers []}
 Pyrotechnics {
  bEnabled=false
  emitterUpdateScript [
   &EmitterState()
   &SpawnRate(SpawnRate=50, Use Spawn Probability=false) {enabled=false, module=SpawnRate002}
   &SpawnRate(
    SpawnRate=&Multiply_Float(
     A=&Multiply_Float(A=40, B=&MakeFloatFromInt(Int=$System.PyroStage1))
     B=$User.Pyrotechnics
    )
    Use Spawn Probability=false
   )
   &SpawnRate(
    SpawnGroup=1
    SpawnRate=&Multiply_Float(
     A=&Multiply_Float(A=65, B=&MakeFloatFromInt(Int=$System.PyroStage2))
     B=$User.Pyrotechnics
    )
    Use Spawn Probability=false
   ) {module=SpawnRate001}
   &SpawnRate(
    SpawnGroup=2
    SpawnRate=&Multiply_Float(
     A=&Multiply_Float(A=65, B=&MakeFloatFromInt(Int=$System.PyroStage3))
     B=$User.Pyrotechnics
    )
    Use Spawn Probability=false
   ) {module=SpawnRate003}
   &SpawnRate(
    SpawnGroup=3
    SpawnRate=&Multiply_Float(
     A=&Multiply_Float(A=65, B=&MakeFloatFromInt(Int=$System.PyroStage4))
     B=$User.Pyrotechnics
    )
    Use Spawn Probability=false
   ) {module=SpawnRate004}
   &SpawnRate(
    SpawnGroup=4
    SpawnRate=&Multiply_Float(
     A=&Multiply_Float(A=65, B=&MakeFloatFromInt(Int=$System.PyroStage5))
     B=$User.Pyrotechnics
    )
    Use Spawn Probability=false
   ) {module=SpawnRate005}
  ]
  particleSpawnScript [
   &PartitionParticles(
    A:NiagaraFloat=&RandomRangeFloat(Minimum=0, Maximum=1)
    B=0.6
    Partition Mode:ENiagara_PartitionMode_v2=::NewEnumerator4
   )
   &InitializeParticle(
    Color=<rgba 1 1 1 0.5>
    Lifetime=&UniformRangedFloat(Minimum=0.65, Maximum=1)
    Mesh Scale=&RandomRangeVector(Minimum=<0.5 0.5 0.8>, Maximum=<0.5 0.6 1>)
    Mesh Scale Mode=::NewEnumerator4
    Sprite Size=&Vector2DFromFloat(Value=&MaskFloat(Mask Source=1, Mask Parameter=$Particles.PartitionParticles.Partition, Mask=0, Masked Float=50, Passthrough Float=5))
    Sprite Size Mode=::NewEnumerator4
    UsePositionOffset=false
    Write Material Random=true
   ) {module=InitializeParticle001}
   &ShapeLocation(
    Cylinder Height=1
    Cylinder Radius=0.1
    MaskSpawnGroup=false
    Offset=&Multiply_Vector(
     A=&MaskPosition(
      Mask Source:NiagaraInt32=0
      Mask Parameter:NiagaraInt32=?
      Mask:NiagaraInt32=0
      Masked Position:NiagaraPosition=<1 1 1>
      Passthrough Position:NiagaraPosition=&MaskPosition(
       Mask Source:NiagaraInt32=1
       Mask Parameter:NiagaraInt32=0
       Mask:NiagaraInt32=&MakeIntFromFloat(Float=&UniformAOrBFloat(A=0, Evaluation Type=::NewEnumerator0, Recalculate Random Each Loop=true, Override Seed=false))
       Masked Position:NiagaraPosition=<1 1 1>
       Passthrough Position:NiagaraPosition=<-1 1 1>
      )
     )
     B=&SelectVectorFromArray(Direct Array Index=$Emitter.SpawnGroup, Vector Selection Array=$System.Fire Array)
    )
    Offset Mode=::NewEnumerator0
    Override Seed=false
    Shape Primitive=::NewEnumerator1
    Surface Only=false
    UseOwnerScale=true
   )
   &AddVelocity(
    Velocity=&RandomRangeVector(Minimum=<0 0 25>, Maximum=<0 0 50>)
    Velocity Speed Scale=&FloatFromCurve(
     FloatCurve=DataInterface() {propertyValues {Curve {keys [<at 0 1 RCIM_Constant>,<at 0.2 2 RCIM_Constant>,<at 0.4 3 RCIM_Constant>,<at 0.6 4 RCIM_Constant>,<at 0.8 5 RCIM_Constant>]}}}
     CurveIndex=&RemapRange(
      Input Value To Remap=&MakeFloatFromInt(Int=$Emitter.SpawnGroup)
      Input Min=0
      Input Max=4
      Output Min=0
      Output Max=1
     )
     Scale Curve=&MaskFloat(Mask Source=1, Mask Parameter=$Particles.PartitionParticles.Partition, Mask=0, Masked Float=1, Passthrough Float=-0.3)
    )
   )
   /SetVariables(Particles.Burned:NiagaraFloat=&ClampFloat(Float=&FloatFromCurve(FloatCurve=DataInterface() {propertyValues {Curve {keys [
          {arriveTangent=7.0415, interpMode=RCIM_Cubic, leaveTangent=7.0415, tangentMode=RCTM_User, time=0.5, value=0}
          {arriveTangent=1.309, interpMode=RCIM_Cubic, leaveTangent=1.309, tangentMode=RCTM_User, time=1, value=1}
         ]}}}))) {isSetParameters=true}
   /SetVariables(
    Particles.Lifetime=&RandomRangeFloat(Minimum=0.75, Maximum=1)
    Particles.Scale=&MaskVector(
     Mask Source=1
     Mask Parameter=$Particles.PartitionParticles.Partition
     Mask=0
     Masked Vector=&RandomRangeVector(Minimum=<0.2 0.6 0.8>, Maximum=<0.3 0.8 1>)
     Passthrough Vector=&Multiply_Vector(
      A=&RandomRangeVector(Minimum=<0.2 0.6 0.8>, Maximum=<0.3 0.8 1>)
      B=&VectorFromFloat(Value=1.1)
     )
    )
   ) {isSetParameters=true}
   /SetVariables(Particles.Lifetime=&Add_Float(A=$Particles.Lifetime, B=&Multiply_Float(A=$Particles.Burned, B=0.3))) {isSetParameters=true}
   &InitialMeshOrientation(
    Mesh Orientation Mode=::NewEnumerator1
    Rotation=&RandomRangeVector(Minimum=<0 0 0>, Maximum=<0 0 0>)
    Use Rotation Vector=true
   ) {enabled=false}
  ]
  particleUpdateScript [
   &ParticleState()
   &ScaleMeshSize(Scale Factor=&MakeVector(
     X=&MakeFloatFromVector(
      Vector=&VectorFromFloat(Value=&FloatFromCurve(
        FloatCurve=DataInterface() {propertyValues {Curve {keys [
            <at 0 0.1 RCIM_Cubic>
            {arriveTangent=2.817, interpMode=RCIM_Cubic, leaveTangent=2.817, tangentMode=RCTM_User, time=0.6, value=0.95}
            {arriveTangent=-2.6611, tangentMode=RCTM_User, time=1, value=0}
           ]}}}
        Scale Curve=2.75
       ))
      Channel=::NewEnumerator0
     )
     Y=&MakeFloatFromVector(
      Vector=&VectorFromFloat(Value=&FloatFromCurve(
        FloatCurve=DataInterface() {propertyValues {Curve {keys [
            <at 0 0.1 RCIM_Cubic>
            {arriveTangent=2.817, interpMode=RCIM_Cubic, leaveTangent=2.817, tangentMode=RCTM_User, time=0.6, value=0.95}
            {arriveTangent=-2.6611, tangentMode=RCTM_User, time=1, value=0}
           ]}}}
        Scale Curve=2.75
       ))
      Channel=::NewEnumerator0
     )
     Z=&MakeFloatFromVector(
      Vector=&VectorFromFloat(Value=&FloatFromCurve(
        FloatCurve=DataInterface() {propertyValues {Curve {keys [
            <at 0 0.1 RCIM_Cubic>
            {arriveTangent=1.3066, interpMode=RCIM_Cubic, leaveTangent=1.3066, tangentMode=RCTM_User, time=0.3558, value=0.95}
            {arriveTangent=-2.6611, tangentMode=RCTM_User, time=1, value=0}
           ]}}}
        Scale Curve=&MaskFloat(
         Mask Source=1
         Mask Parameter=$Particles.PartitionParticles.Partition
         Mask=0
         Masked Float=&FloatFromCurve(
          FloatCurve=DataInterface() {propertyValues {Curve {keys [<at 0 1.5 RCIM_Constant>,<at 0.2 2.5 RCIM_Constant>,<at 0.4 3 RCIM_Constant>,<at 0.6 3.5 RCIM_Constant>,<at 0.8 4 RCIM_Constant>]}}}
          CurveIndex=&RemapRange(
           Input Value To Remap=&MakeFloatFromInt(Int=$Emitter.SpawnGroup)
           Input Min=0
           Input Max=4
           Output Min=0
           Output Max=1
          )
          Scale Curve=1.3
         )
         Passthrough Float=1
        )
       ))
      Channel=::NewEnumerator0
     )
    ))
   &CurlNoiseForce(Noise Frequency=0.1, Noise Strength=50, Pan Noise=true, Pan Noise Field=<0 0 1>) {enabled=false}
   &GravityForce(Gravity=<0 0 150>) {enabled=false}
   &Drag(Use Linear Drag=true, Use Rotational Drag=true)
   &SolveForcesAndVelocity(Clamp Velocity=true, Limit Acceleration=false, Speed Limit=500)
   &ScaleMeshSize(Scale Factor=&Multiply_VectorByFloat(
     Vector=$Particles.Initial.Scale
     Float=&FloatFromCurve(
      FloatCurve=DataInterface() {propertyValues {Curve {keys [
          <at 0 0.25 RCIM_Cubic>
          {arriveTangent=-0.565, interpMode=RCIM_Cubic, leaveTangent=-0.565, tangentMode=RCTM_User, time=0.2, value=1}
          {interpMode=RCIM_Cubic, tangentMode=RCTM_User, time=1, value=0.75}
         ]}}}
      Scale Curve=&RandomRangeFloat(Minimum=0.35, Maximum=0.6)
     )
    )) {module=ScaleMeshSize001}
   &SubUVAnimation(
    Start Frame:NiagaraInt32=&RandomRangeInt(Minimum=0, Maximum=16)
    SubUV Loop Count:NiagaraFloat=&RandomRangeFloat(Minimum=1, Maximum=1.25)
   )
   &DynamicMaterialParameters(
    Index 0 Param 1=&Multiply_Float(
     A=&RemapRange(Input Value To Remap=$Particles.Burned, Input Min=0, Input Max=1, Output Min=0.25, Output Max=1)
     B=&RandomRangeFloat(Minimum=1, Maximum=0.2)
    )
    Index 0 Param 2=&RandomRangeFloat(Minimum=0, Maximum=2)
    Index 0 Param 3=8
    Index 0 Param 4=$Particles.SubImageIndex
    Index 1 Param 1=&RandomRangeFloat(Minimum=-1, Maximum=1)
    Index 1 Param 2=&RandomRangeFloat(Minimum=0, Maximum=3)
    Index 1 Param 3=&FloatFromCurve(
     FloatCurve=DataInterface() {propertyValues {Curve {keys [
         {arriveTangent=4.1785, interpMode=RCIM_Cubic, leaveTangent=4.1785, tangentMode=RCTM_User, time=0, value=0}
         {arriveTangent=0.9995, interpMode=RCIM_Cubic, leaveTangent=0.9995, tangentMode=RCTM_User, time=0.2223, value=0.6825}
         {arriveTangent=-0.0405, interpMode=RCIM_Cubic, leaveTangent=-0.0405, tangentMode=RCTM_User, time=0.5, value=0.5}
         <at 1 0.25>
        ]}}}
     Scale Curve=&RemapRange(Input Value To Remap=$Particles.Burned, Input Min=0, Input Max=1, Output Min=1, Output Max=0.5)
    )
    Module.Param0WriteEnabled=true
    Module.Param0WriteEnabled1=true
    Module.Param1WriteEnabled=true
    Module.Param1WriteEnabled1=true
    Module.Param2WriteEnabled=true
    Module.Param2WriteEnabled1=true
    Module.Param3WriteEnabled=true
    Module.Param3WriteEnabled1=false
    Write Parameter Index 1=true
   )
   &ScaleColor(
    Scale Alpha=&FloatFromCurve(
     FloatCurve=DataInterface() {propertyValues {Curve {keys [
         {arriveTangent=9.0725, interpMode=RCIM_Cubic, leaveTangent=9.0725, tangentMode=RCTM_User, time=0, value=0}
         {arriveTangent=0.1459, interpMode=RCIM_Cubic, leaveTangent=0.1459, tangentMode=RCTM_User, time=0.1408, value=0.9847}
         {arriveTangent=-0.0405, interpMode=RCIM_Cubic, leaveTangent=-0.0405, tangentMode=RCTM_User, time=0.7, value=0}
        ]}}}
     Scale Curve=&MaskFloat(Mask Source=1, Mask Parameter=$Particles.Initial.PartitionParticles.Partition, Mask=0, Masked Float=1, Passthrough Float=10)
    )
    Scale RGB=&Multiply_VectorByFloat(
     Vector=<2 2 2>
     Float=&FloatFromCurve(
      FloatCurve=DataInterface() {propertyValues {Curve {keys [
          {arriveTangent=9.0725, interpMode=RCIM_Cubic, leaveTangent=9.0725, tangentMode=RCTM_User, time=0, value=0}
          {arriveTangent=-2.5249, interpMode=RCIM_Cubic, leaveTangent=-2.5249, tangentMode=RCTM_User, time=0.1408, value=0.9847}
          {arriveTangent=-0.8514, interpMode=RCIM_Cubic, leaveTangent=-0.8514, tangentMode=RCTM_User, time=0.5, value=0.25}
          {arriveTangent=-0.0405, interpMode=RCIM_Cubic, leaveTangent=-0.0405, tangentMode=RCTM_User, time=1, value=0}
         ]}}}
      Scale Curve=&MaskFloat(Mask Source=1, Mask Parameter=$Particles.Initial.PartitionParticles.Partition, Mask=0, Masked Float=15, Passthrough Float=30)
     )
    )
    ScaleA=true
    ScaleRGB=true
   ) {module=ScaleColor001}
   &CameraOffset(Camera Offset Amount=&MaskFloat(Mask Source=1, Mask Parameter=$Particles.Initial.PartitionParticles.Partition, Mask=0, Masked Float=2, Passthrough Float=15))
  ]
  rendererClasses [&NiagaraMeshRendererProperties]
  renderers [{rendererClass=&NiagaraMeshRendererProperties, rendererIndex=0}]
 }
 Pyrotechnics_Highlights {
  bEnabled=false
  emitterUpdateScript [
   &EmitterState()
   &SpawnRate(
    SpawnRate=&Multiply_Float(
     A=&Multiply_Float(A=15, B=&MakeFloatFromInt(Int=$System.PyroStage1))
     B=$User.Pyrotechnics
    )
    Use Spawn Probability=false
   ) {module=SpawnRate001}
   &SpawnRate(
    SpawnGroup=1
    SpawnRate=&Multiply_Float(
     A=&Multiply_Float(A=15, B=&MakeFloatFromInt(Int=$System.PyroStage2))
     B=$User.Pyrotechnics
    )
    Use Spawn Probability=false
   ) {module=SpawnRate002}
   &SpawnRate(
    SpawnGroup=2
    SpawnRate=&Multiply_Float(
     A=&Multiply_Float(A=15, B=&MakeFloatFromInt(Int=$System.PyroStage3))
     B=$User.Pyrotechnics
    )
    Use Spawn Probability=false
   ) {module=SpawnRate003}
   &SpawnRate(
    SpawnGroup=3
    SpawnRate=&Multiply_Float(
     A=&Multiply_Float(A=20, B=&MakeFloatFromInt(Int=$System.PyroStage4))
     B=$User.Pyrotechnics
    )
    Use Spawn Probability=false
   ) {module=SpawnRate004}
   &SpawnRate(
    SpawnGroup=4
    SpawnRate=&Multiply_Float(
     A=&Multiply_Float(A=25, B=&MakeFloatFromInt(Int=$System.PyroStage5))
     B=$User.Pyrotechnics
    )
    Use Spawn Probability=false
   ) {module=SpawnRate005}
  ]
  particleSpawnScript [
   &InitializeParticle(
    Color Maximum=<rgba 2 0.2507 0.0513 1>
    Color Minimum=<rgba 2 0.4314 0.1333 1>
    Color Mode=::NewEnumerator2
    Lifetime=&UniformRangedFloat(Minimum=0.75, Maximum=1)
    Sprite Rotation Mode=::NewEnumerator2
    Sprite Size=<150 150>
    Sprite Size Mode=::NewEnumerator4
    UsePositionOffset=false
    Write Material Random=true
   ) {module=InitializeParticle001}
   &ShapeLocation(
    Cylinder Height=1
    Cylinder Radius=0.1
    MaskSpawnGroup=false
    Offset=&Multiply_Vector(
     A=&MaskPosition(
      Mask Source:NiagaraInt32=0
      Mask Parameter:NiagaraInt32=?
      Mask:NiagaraInt32=0
      Masked Position:NiagaraPosition=<1 1 1>
      Passthrough Position:NiagaraPosition=&MaskPosition(
       Mask Source:NiagaraInt32=1
       Mask Parameter:NiagaraInt32=0
       Mask:NiagaraInt32=&MakeIntFromFloat(Float=&UniformAOrBFloat(A=0, Evaluation Type=::NewEnumerator0, Recalculate Random Each Loop=true, Override Seed=false))
       Masked Position:NiagaraPosition=<1 1 1>
       Passthrough Position:NiagaraPosition=<-1 1 1>
      )
     )
     B=&SelectVectorFromArray(Direct Array Index=$Emitter.SpawnGroup, Vector Selection Array=$System.Fire Array)
    )
    Offset Mode=::NewEnumerator0
    Override Seed=false
    Shape Primitive=::NewEnumerator1
    Surface Only=false
    UseOwnerScale=true
   )
   &AddVelocity(
    Velocity=&RandomRangeVector(Minimum=<0 0 25>, Maximum=<1 0 50>)
    Velocity Speed Scale=&FloatFromCurve(
     FloatCurve=DataInterface() {propertyValues {Curve {keys [<at 0 1 RCIM_Constant>,<at 0.2 2 RCIM_Constant>,<at 0.4 3 RCIM_Constant>,<at 0.6 3.5 RCIM_Constant>,<at 0.8 4 RCIM_Constant>]}}}
     CurveIndex=&RemapRange(
      Input Value To Remap=&MakeFloatFromInt(Int=$Emitter.SpawnGroup)
      Input Min=0
      Input Max=4
      Output Min=0
      Output Max=0.3
     )
     Scale Curve=2
    )
   )
  ]
  particleUpdateScript [
   &ParticleState()
   /Niagara/Modules/Update/SubUV/SubUVAnimation(Number Of Frames=&FloatFromCurve(FloatCurve=DataInterface() {propertyValues {Curve {keys [{arriveTangent=64.0636, interpMode=RCIM_Cubic, leaveTangent=64.0636, tangentMode=RCTM_User, time=0, value=43}, <at 0.9 62>]}}})) {module=SubUVAnimation001}
   &DynamicMaterialParameters(
    Index 0 Param 1=&FloatFromCurve(FloatCurve=DataInterface() {propertyValues {Curve {keys [<at 0 1>]}}})
    Index 0 Param 2=0.5
    Index 0 Param 3=&FloatFromCurve(FloatCurve=DataInterface() {propertyValues {Curve {keys [<at 0 0.5 RCIM_Cubic>, {arriveTangent=1.5, interpMode=RCIM_Cubic, leaveTangent=1.5, time=0.5235, value=0.5}, <at 1 2>]}}})
    Module.Param0WriteEnabled=true
    Module.Param1WriteEnabled=false
    Module.Param2WriteEnabled=true
    Module.Param3WriteEnabled=false
   ) {module=DynamicMaterialParameters001}
   &CurlNoiseForce(Noise Frequency=0.1, Noise Strength=50, Pan Noise=true, Pan Noise Field=<0 0 1>)
   &GravityForce(Gravity=<0 0 150>)
   &Drag(Use Linear Drag=true, Use Rotational Drag=true) {module=Drag001}
   &SolveForcesAndVelocity(Clamp Velocity=true, Limit Acceleration=false, Speed Limit=500)
   &ScaleSpriteSizeBySpeed(Max Scale Factor=<1 10>) {enabled=false}
   &ScaleSpriteSize(Scale Factor=&Multiply_Vector2D_ByFloat(
     Float=&FloatFromCurve(
      FloatCurve=DataInterface() {propertyValues {Curve {keys [{interpMode=RCIM_Cubic, tangentMode=RCTM_User, time=0.1, value=0.1}, {interpMode=RCIM_Cubic, tangentMode=RCTM_User, time=1, value=1}]}}}
      Scale Curve=0.8
     )
     Vector2D=<1 1>
    )) {module=ScaleSpriteSize001}
   &ScaleColor(
    Scale Alpha=&FloatFromCurve(
     FloatCurve=DataInterface() {propertyValues {Curve {keys [
         {arriveTangent=11.0384, interpMode=RCIM_Cubic, leaveTangent=11.0384, tangentMode=RCTM_User, time=0, value=0}
         {arriveTangent=0.1459, interpMode=RCIM_Cubic, leaveTangent=0.1459, tangentMode=RCTM_User, time=0.1408, value=0.9847}
         {arriveTangent=-0.0737, interpMode=RCIM_Cubic, leaveTangent=-0.0737, tangentMode=RCTM_User, time=0.7, value=0}
        ]}}}
     Scale Curve=0.1
    )
    Scale RGB=&Multiply_VectorByFloat(
     Vector=<2 2 2>
     Float=&FloatFromCurve(
      FloatCurve=DataInterface() {propertyValues {Curve {keys [
          {arriveTangent=9.0725, interpMode=RCIM_Cubic, leaveTangent=9.0725, tangentMode=RCTM_User, time=0, value=0}
          {arriveTangent=-2.5249, interpMode=RCIM_Cubic, leaveTangent=-2.5249, tangentMode=RCTM_User, time=0.1408, value=0.9847}
          {arriveTangent=-0.8514, interpMode=RCIM_Cubic, leaveTangent=-0.8514, tangentMode=RCTM_User, time=0.5, value=0.25}
          {arriveTangent=-0.0405, interpMode=RCIM_Cubic, leaveTangent=-0.0405, tangentMode=RCTM_User, time=1, value=0}
         ]}}}
      Scale Curve=25
     )
    )
    ScaleA=true
    ScaleRGB=true
   ) {module=ScaleColor001}
  ]
  rendererClasses [&NiagaraSpriteRendererProperties]
  renderers [{rendererClass=&NiagaraSpriteRendererProperties}]
 }
}
summary {
 systemName=NS_Emote_SummitReedMolt
 emitters [
  {
   emitterName=Ambient_Condensed
   rendererClasses [&NiagaraSpriteRendererProperties, &NiagaraMeshRendererProperties]
  }
  {bEnabled=false, emitterName=Embers, rendererClasses [&NiagaraSpriteRendererProperties]}
  {bEnabled=false, emitterName=Pyrotechnics, rendererClasses [&NiagaraMeshRendererProperties]}
  {bEnabled=false, emitterName=Pyrotechnics_Highlights, rendererClasses [&NiagaraSpriteRendererProperties]}
  {emitterName=Flashing_Lights, rendererClasses [&NiagaraSpriteRendererProperties]}
  {emitterName=Front_Light, rendererClasses []}
  {emitterName=Back_Light, rendererClasses []}
  {emitterName=Leg_Lights, rendererClasses []}
 ]
 userVariables [
  User.bIsFrontEnd {type=NiagaraBool}
  User.bIsFrontEndPreview {type=NiagaraBool}
  User.Pyrotechnics {type=NiagaraFloat, defaultValue=0.5, description=""}
 ]
}
